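# Niagara impact effect for the AnvilAlarmMoth MicStand pickaxe, layering a purple-cyan flash, flares, smoke, sparks, cube ring and surface debris.
# Core_Flash packs several looks into one emitter through spawn-group switches feeding mesh, sprite and light renderers, with time-of-day brightness scaling.
# FN_PhysMatDebris takes velocity and gravity from User.PhysicalSurface.
# System Global Scale drives most sizes, front-end switches retune several emitters for preview, and the Cubes emitter is disabled.
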NiagaraSystem /SparksCosmetics/Instrument/MicStand/AnvilAlarmMoth/Effects/NS_Pickaxe_MicStand_AnvilAlarmMoth_Impact
dependencies {
 usedDataInterfaces [
  &NiagaraDataInterfaceCurve
  &NiagaraDataInterfaceVectorCurve
  &NiagaraDataInterfaceColorCurve
  &NiagaraDataInterfaceCamera
 ]
 usedDynamicInputs [&MakeIntFromFloat]
 usedModules [
  &self:SystemSpawnScript.SystemScriptSource.SystemScriptGraph.NiagaraNodeAssignment_0.SetVariables_#1
  &self:SystemSpawnScript.SystemScriptSource.SystemScriptGraph.NiagaraNodeAssignment_1.SetVariables_#2
  &SystemState
 ]
}
emitters {
 Core_Flash {
  emitterUpdateScript [
   &EmitterState()
   &SpawnBurst_Instantaneous(Spawn Count=1, Use Spawn Probability=false) {enabled=false}
   &SpawnBurst_Instantaneous(Spawn Count=&RandomRangeInt(Minimum=10, Maximum=11), Spawn Group=1, Use Spawn Probability=false) {module=SpawnBurst_Instantaneous001}
   &SpawnBurst_Instantaneous(Spawn Count=5, Spawn Group=2, Use Spawn Probability=false, UseLoopCountLimit=false) {enabled=false, module=SpawnBurst_Instantaneous002}
   &SpawnBurst_Instantaneous(Spawn Group=3, Use Spawn Probability=false, UseLoopCountLimit=false) {module=SpawnBurst_Instantaneous003}
   &SpawnBurst_Instantaneous(Spawn Group=4, Use Spawn Probability=false, UseLoopCountLimit=false) {module=SpawnBurst_Instantaneous004}
   &SpawnBurst_Instantaneous(Spawn Group=5, Use Spawn Probability=false, UseLoopCountLimit=false) {module=SpawnBurst_Instantaneous005}
   &SpawnBurst_Instantaneous(Spawn Group=6, Use Spawn Probability=false, UseLoopCountLimit=false) {module=SpawnBurst_Instantaneous006}
  ]
  particleSpawnScript [
   /SetVariables(Particles.VisibilityTag=&SpawnGroup()) {isSetParameters=true}
   &InitializeParticle(
    Color=&MergeHelper_ColorSwitch(
     0=<rgba 1.5 1.5 1.5 1>
     1=<rgba 1.1 1.1 1.1 1>
     2=<rgba 0.2865 0.2865 0.2865 0.8582>
     3=<rgba 1 1 1 1>
     4=<rgba 1 1 1 1>
     5=<rgba 0.2995 0.2995 0.2995 0.9>
     6=<rgba 1.1549 0.3711 1.5 1>
     Index=&SpawnGroup()
     Num=::NewEnumerator5
    )
    Lifetime=&MergeHelper_FloatSwitch(
     0=0.22
     1=&RandomRangeFloat(Minimum=0.15, Maximum=0.2)
     2=0.17
     3=0.18
     4=0.22
     5=0.25
     6=0.2
     Index=&SpawnGroup()
     Num=::NewEnumerator5
    )
    Mesh Scale=&Multiply_VectorByFloat(
     Vector=&Multiply_VectorByFloat(Vector=<1 2 2>, Float=&RandomRangeFloat(Minimum=0.8, Maximum=0.82))
     Float=$System.Global Scale
    )
    Mesh Scale Mode=::NewEnumerator4
    Sprite Size=&Multiply_Vector2D_ByFloat(
     Float=$System.Global Scale
     Vector2D=&MergeHelper_Vector2Switch(
      0=<0 0>
      1=&RandomRangeVector2D(Minimum=<9 50>, Maximum=<13 65>)
      2=&RandomRangeVector2D(Minimum=<30 45>, Maximum=<45 65>)
      3=&Vector2DFromFloat(Value=200)
      4=<1 1>
      5=&Vector2DFromFloat(Value=200)
      6=&Vector2DFromFloat(Value=70)
      Index=$Emitter.SpawnGroup
      Num=::NewEnumerator5
     )
    )
    Sprite Size Mode=::NewEnumerator4
    UsePositionOffset=false
    Write Material Random=true
   )
   &ShapeLocation(
    MaskSpawnGroup=false
    Override Seed=false
    Rotation Angle:NiagaraFloat=&RandomRangeFloat(Minimum=0, Maximum=1)
    Rotation Mode=::NewEnumerator1
    Sphere Radius=&MergeHelper_FloatSwitch(
     0=0
     1=&Multiply_Float(A=2, B=$System.Global Scale)
     2=&Multiply_Float(A=6, B=$System.Global Scale)
     3=0
     4=0
     5=0
     Index=&SpawnGroup()
     Num=::NewEnumerator4
    )
    UseOwnerScale=true
   )
   &DynamicMaterialParameters(
    Index 0 Param 2=45
    Index 0 Param 3=1
    Index 0 Param 4=1
    Module.Param0WriteEnabled=false
    Module.Param1WriteEnabled=true
    Module.Param2WriteEnabled=true
    Module.Param3WriteEnabled=true
   ) {module=DynamicMaterialParameters001}
   &InitialMeshOrientation(
    Rotation=&RandomRangeVector(Minimum=<-1 -1 -1>, Maximum=<1 1 1>)
    Use Orientation Vector=true
    Use Rotation Vector=true
   )
   &CameraOffset(Camera Offset Amount=&MergeHelper_FloatSwitch(0=20, 1=2, 2=10, 3=10, 4=0, 5=25, 6=20, Index=&SpawnGroup(), Num=::NewEnumerator5))
   &AddVelocityFromPoint(
    Module.Use Falloff Distance=false
    Velocity Strength=&MergeHelper_FloatSwitch(0=0, 1=2, 2=125, 3=0, 4=0, Index=&SpawnGroup(), Num=::NewEnumerator3)
   )
   /SetVariables(Particles.Flare SUBUV:NiagaraFloat=&RandomRangeFloat(Minimum=1, Maximum=3)) {isSetParameters=true}
  ]
  particleUpdateScript [
   &ParticleState()
   &DynamicMaterialParameters(
    Index 0 Param 1=&FloatFromCurve(FloatCurve=DataInterface() {propertyValues {Curve {keys [<at 0.2 0 RCIM_Cubic>, {arriveTangent=3.6451, interpMode=RCIM_Cubic, leaveTangent=3.6451, tangentMode=RCTM_User, time=1, value=1}]}}})
    Module.Index0Float0Write=true
    Module.Index0Float1Write=true
    Module.Index0Float2Write=true
    Module.Index0Float3Write=true
   )
   &ScaleMeshSize(Scale Factor=&VectorFromFloat(Value=&FloatFromCurve(
      FloatCurve=DataInterface() {propertyValues {Curve {
         defaultValue=0.5
         keys [
          {arriveTangent=2.9597, interpMode=RCIM_Cubic, leaveTangent=2.9597, tangentMode=RCTM_User, time=0, value=0}
          {arriveTangent=0.721, interpMode=RCIM_Cubic, leaveTangent=0.721, tangentMode=RCTM_User, time=1, value=1}
         ]
        }}}
      Scale Curve=0.7
     )))
   &ScaleColor(
    Scale Alpha=&MergeHelper_FloatSwitch(
     0=&FloatFromCurve(FloatCurve=DataInterface() {propertyValues {Curve {keys [<at 0 1>,<at 1 0>]}}})
     1=&FloatFromCurve(FloatCurve=DataInterface() {propertyValues {Curve {
         defaultValue=1
         keys [
          {arriveTangent=-0.592, interpMode=RCIM_Cubic, leaveTangent=-0.592, tangentMode=RCTM_User, time=0, value=1}
          {interpMode=RCIM_Cubic, tangentMode=RCTM_User, time=1, value=0}
         ]
        }}})
     2=&FloatFromCurve(FloatCurve=DataInterface() {propertyValues {Curve {
         defaultValue=1
         keys [
          {arriveTangent=-0.5163, interpMode=RCIM_Cubic, leaveTangent=-0.5163, tangentMode=RCTM_User, time=0, value=1}
          {interpMode=RCIM_Cubic, tangentMode=RCTM_User, time=1, value=0}
         ]
        }}})
     3=&FloatFromCurve(
      FloatCurve=DataInterface() {propertyValues {Curve {
         defaultValue=1
         keys [
          {interpMode=RCIM_Cubic, tangentMode=RCTM_User, time=0, value=1}
          {arriveTangent=-1.2989, interpMode=RCIM_Cubic, leaveTangent=-1.2989, tangentMode=RCTM_User, time=1, value=0}
         ]
        }}}
      Scale Curve=&FN_FrontEndFloatModifier(
       Gameplay=&FloatFromCurve(
        FloatCurve=DataInterface() {propertyValues {Curve {keys [
            <at 0 0.65 RCIM_Cubic>
            {arriveTangent=0.875, interpMode=RCIM_Cubic, leaveTangent=0.875, time=0.1, value=0.7}
            {arriveTangentWeight=0.1329, interpMode=RCIM_Cubic, leaveTangentWeight=0.0759, tangentMode=RCTM_User, time=0.4, value=1}
            {arriveTangentWeight=0.0801, interpMode=RCIM_Cubic, leaveTangentWeight=0.1002, tangentMode=RCTM_User, time=0.6, value=1}
            {arriveTangent=-0.8778, interpMode=RCIM_Cubic, leaveTangent=-0.8778, time=0.9, value=0.7}
            <at 0.9987 0.65 RCIM_Cubic>
           ]}}}
        CurveIndex=Linked() {linkedVariable {name=NPC.FortniteNPC.TimeOfDay_0-1, type=NiagaraFloat}}
       )
       Front End=0.8
      )
     )
     4=1
     5=&FloatFromCurve(
      FloatCurve=DataInterface() {propertyValues {Curve {
         defaultValue=1
         keys [
          {interpMode=RCIM_Cubic, tangentMode=RCTM_User, time=0, value=1}
          {arriveTangent=-2.1802, interpMode=RCIM_Cubic, leaveTangent=-2.1802, tangentMode=RCTM_User, time=1, value=0}
         ]
        }}}
      Scale Curve=&FloatFromCurve(
       FloatCurve=DataInterface() {propertyValues {Curve {
          defaultValue=1
          keys [
           {arriveTangent=0.664, interpMode=RCIM_Cubic, leaveTangent=0.664, tangentMode=RCTM_User, time=0, value=0.45}
           {interpMode=RCIM_Cubic, tangentMode=RCTM_User, time=0.5, value=1}
           {tangentMode=RCTM_User, time=1, value=0.45}
          ]
         }}}
       CurveIndex=Linked() {linkedVariable {name=NPC.FortniteNPC.TimeOfDay_0-1, type=NiagaraFloat}}
       Scale Curve=1.2
      )
     )
     6=&FloatFromCurve(FloatCurve=DataInterface() {propertyValues {Curve {keys [
          <at 0.0045 1 RCIM_Cubic>
          {arriveTangent=-2.0573, arriveTangentWeight=0.5281, interpMode=RCIM_Cubic, leaveTangent=-2.0573, tangentMode=RCTM_User, time=1, value=0}
         ]}}})
     Index=&SpawnGroup()
     Num=::NewEnumerator5
    )
    Scale RGB=&MergeHelper_VectorSwitch(
     0=&Multiply_VectorByFloat(
      Vector=&MakeVectorFromLinearColorRGB(Color=&ColorFromCurve(ColorCurve=DataInterface() {propertyValues {
          AlphaCurve {keys []}
          BlueCurve {keys [<at 0.032 2.8984 RCIM_Cubic>,<at 0.8557 0.334 RCIM_Cubic>]}
          GreenCurve {keys [<at 0.032 5.7139 RCIM_Cubic>,<at 0.8557 0.4939 RCIM_Cubic>]}
          RedCurve {keys [<at 0.032 7 RCIM_Cubic>,<at 0.8557 0.7995 RCIM_Cubic>]}
         }}))
      Float=&FloatFromCurve(
       FloatCurve=DataInterface() {propertyValues {Curve {keys [<at 0 4>,<at 0.2 1.5>,<at 1 1>]}}}
       Scale Curve=&FloatFromCurve(
        FloatCurve=DataInterface() {propertyValues {Curve {keys [
            <at 0.1 0.8 RCIM_Cubic>
            {arriveTangentWeight=0.1329, interpMode=RCIM_Cubic, leaveTangentWeight=0.0759, tangentMode=RCTM_User, time=0.35, value=1}
            {arriveTangentWeight=0.0801, interpMode=RCIM_Cubic, leaveTangentWeight=0.1002, tangentMode=RCTM_User, time=0.65, value=1}
            <at 0.9 0.8 RCIM_Cubic>
           ]}}}
        CurveIndex=Linked() {linkedVariable {name=NPC.FortniteNPC.TimeOfDay_0-1, type=NiagaraFloat}}
       )
      )
     )
     1=&Multiply_VectorByFloat(
      Vector=&MakeVectorFromLinearColorRGB(Color=&ColorFromCurve(ColorCurve=DataInterface() {propertyValues {
          AlphaCurve {keys []}
          BlueCurve {keys [<at 0 6 RCIM_Cubic>,<at 0.6648 2>,<at 0.9945 1>]}
          GreenCurve {keys [<at 0 0.6562 RCIM_Cubic>,<at 0.6648 0.4216>,<at 0.9945 0>]}
          RedCurve {keys [<at 0 5.9624 RCIM_Cubic>,<at 0.6648 0.9807>,<at 0.9945 0.6519>]}
         }}))
      Float=&FloatFromCurve(
       FloatCurve=DataInterface() {propertyValues {Curve {keys [
           <at 0.1 0.8 RCIM_Cubic>
           {arriveTangentWeight=0.1329, interpMode=RCIM_Cubic, leaveTangentWeight=0.0759, tangentMode=RCTM_User, time=0.45, value=1}
           {arriveTangentWeight=0.0801, interpMode=RCIM_Cubic, leaveTangentWeight=0.1002, tangentMode=RCTM_User, time=0.55, value=1}
           <at 0.9 0.8 RCIM_Cubic>
          ]}}}
       CurveIndex=Linked() {linkedVariable {name=NPC.FortniteNPC.TimeOfDay_0-1, type=NiagaraFloat}}
      )
     )
     2=&VectorFromFloat(Value=&FN_FrontEndFloatModifier(Gameplay=1, Front End=0.9))
     3=&Multiply_VectorByFloat(
      Vector=&MakeVectorFromLinearColorRGB(Color=&ColorFromCurve(ColorCurve=DataInterface() {propertyValues {
          AlphaCurve {keys []}
          BlueCurve {keys [<at 0 1.5 RCIM_Cubic>,<at 0.2717 1>,<at 0.6802 0.985>,<at 1 0.647>]}
          GreenCurve {keys [<at 0 0.2344 RCIM_Cubic>,<at 0.2717 0.0703>,<at 0.6802 0.3015>,<at 1 0.6556>]}
          RedCurve {keys [<at 0 1.1131 RCIM_Cubic>,<at 0.2717 0.7158>,<at 0.6802 0.236>,<at 1 0>]}
         }}))
      Float=&FloatFromCurve(
       FloatCurve=DataInterface() {propertyValues {Curve {keys [
           <at 0.1 0.75 RCIM_Cubic>
           {arriveTangentWeight=0.1329, interpMode=RCIM_Cubic, leaveTangentWeight=0.0759, tangentMode=RCTM_User, time=0.45, value=1}
           {arriveTangentWeight=0.0801, interpMode=RCIM_Cubic, leaveTangentWeight=0.1002, tangentMode=RCTM_User, time=0.55, value=1}
           <at 0.9 0.75 RCIM_Cubic>
          ]}}}
       CurveIndex=Linked() {linkedVariable {name=NPC.FortniteNPC.TimeOfDay_0-1, type=NiagaraFloat}}
       Scale Curve=&FloatFromCurve(FloatCurve=DataInterface() {propertyValues {Curve {keys [{arriveTangent=-6.5809, interpMode=RCIM_Cubic, leaveTangent=-6.5809, tangentMode=RCTM_User, time=0.1356, value=3}, <at 1 0.5 RCIM_Cubic>]}}})
      )
     )
     4=<1 1 1>
     5=&MakeVectorFromLinearColorRGB(Color=&ColorFromCurve(ColorCurve=DataInterface() {propertyValues {
         AlphaCurve {keys []}
         BlueCurve {keys [<at 0.0724 0.9635 RCIM_Cubic>,<at 0.459 0.651>,<at 0.9556 0.647>]}
         GreenCurve {keys [<at 0.0724 0.1054 RCIM_Cubic>,<at 0.459 0.1373>,<at 0.9556 0.6556>]}
         RedCurve {keys [<at 0.0724 0.9575 RCIM_Cubic>,<at 0.459 0.3193>,<at 0.9556 0>]}
        }}))
     6=&Multiply_VectorByFloat(
      Vector=&MakeVectorFromLinearColorRGB(Color=&ColorFromCurve(ColorCurve=DataInterface() {propertyValues {
          AlphaCurve {keys []}
          BlueCurve {keys [<at 0 3>,<at 0.8917 2.1083>,<at 1 2>]}
          GreenCurve {keys [<at 0 0.4482>,<at 0.8917 0.1537>,<at 1 0.6175>]}
          RedCurve {keys [<at 0 2.22>,<at 0.8917 1.3997>,<at 1 0.3021>]}
         }}))
      Float=&FloatFromCurve(
       FloatCurve=DataInterface() {propertyValues {Curve {keys [<at 0 4>,<at 0.4 1.5>,<at 1 1>]}}}
       Scale Curve=&FloatFromCurve(
        FloatCurve=DataInterface() {propertyValues {Curve {keys [
            <at 0.1 0.8 RCIM_Cubic>
            {arriveTangentWeight=0.1329, interpMode=RCIM_Cubic, leaveTangentWeight=0.0759, tangentMode=RCTM_User, time=0.35, value=1}
            {arriveTangentWeight=0.0801, interpMode=RCIM_Cubic, leaveTangentWeight=0.1002, tangentMode=RCTM_User, time=0.65, value=1}
            <at 0.9 0.8 RCIM_Cubic>
           ]}}}
        CurveIndex=Linked() {linkedVariable {name=NPC.FortniteNPC.TimeOfDay_0-1, type=NiagaraFloat}}
       )
      )
     )
     Index=&SpawnGroup()
     Num=::NewEnumerator5
    )
    ScaleA=true
    ScaleRGB=true
   )
   &ScaleSpriteSize(
    Scale Factor=&MergeHelper_Vector2Switch(
     0=<0 0>
     1=&MakeVector2D(
      X=&FloatFromCurve(FloatCurve=DataInterface() {propertyValues {Curve {
          defaultValue=1
          keys [
           {interpMode=RCIM_Cubic, tangentMode=RCTM_User, time=0, value=0}
           {arriveTangentWeight=0.2236, interpMode=RCIM_Cubic, leaveTangentWeight=0.1491, tangentMode=RCTM_User, time=0.7, value=1}
           {arriveTangent=-2.2754, interpMode=RCIM_Cubic, leaveTangent=-2.2754, tangentMode=RCTM_User, time=1, value=0.5}
          ]
         }}})
      Y=&FloatFromCurve(FloatCurve=DataInterface() {propertyValues {Curve {
          defaultValue=1
          keys [
           {arriveTangent=3.143, interpMode=RCIM_Cubic, leaveTangent=3.143, tangentMode=RCTM_User, time=0, value=0}
           {arriveTangent=0.8537, interpMode=RCIM_Cubic, leaveTangent=0.8537, tangentMode=RCTM_User, time=1, value=1}
          ]
         }}})
     )
     2=&Vector2DFromFloat(Value=&FloatFromCurve(FloatCurve=DataInterface() {propertyValues {Curve {
          defaultValue=1
          keys [
           {arriveTangent=0.9682, interpMode=RCIM_Cubic, leaveTangent=0.9682, tangentMode=RCTM_User, time=0, value=0}
           {interpMode=RCIM_Cubic, tangentMode=RCTM_User, time=1, value=1}
          ]
         }}}))
     3=&Multiply_Vector2D_ByFloat(
      Float=&FloatFromCurve(FloatCurve=DataInterface() {propertyValues {Curve {
          defaultValue=0
          keys [
           {interpMode=RCIM_Cubic, tangentMode=RCTM_User, time=0, value=1}
           {arriveTangent=-1.4808, interpMode=RCIM_Cubic, leaveTangent=-1.4808, tangentMode=RCTM_User, time=1, value=0}
          ]
         }}})
      Vector2D=<1 1>
     )
     4=<0 0>
     5=&Vector2DFromFloat(Value=&FloatFromCurve(FloatCurve=DataInterface() {propertyValues {Curve {
          defaultValue=0
          keys [
           {arriveTangent=2.4894, interpMode=RCIM_Cubic, leaveTangent=2.4894, tangentMode=RCTM_User, time=0, value=0}
           {interpMode=RCIM_Cubic, tangentMode=RCTM_User, time=1, value=1}
          ]
         }}}))
     6=&Multiply_Vector2D_ByFloat(
      Float=&FloatFromCurve(FloatCurve=DataInterface() {propertyValues {Curve {
          defaultValue=0
          keys [
           {arriveTangent=2.0604, interpMode=RCIM_Cubic, leaveTangent=2.0604, tangentMode=RCTM_User, time=0, value=0.2}
           {arriveTangent=0.201, interpMode=RCIM_Cubic, leaveTangent=0.201, tangentMode=RCTM_User, time=1, value=1}
          ]
         }}})
      Vector2D=<1 1>
     )
     Index=&SpawnGroup()
     Num=::NewEnumerator5
    )
    Scale Sprite Size Mode=::NewEnumerator2
   )
   &SolveForcesAndVelocity(Clamp Velocity=false, Limit Acceleration=false)
   /SetVariables(Particles.Flare Dyn 1:Vector4f=&MakeVector4(
     X=&FloatFromCurve(FloatCurve=DataInterface() {propertyValues {Curve {keys [<at 0 1 RCIM_Cubic>,<at 0.8 0 RCIM_Cubic>]}}})
     Y=2
     Z=1
     W=1
    )) {isSetParameters=true}
   &FN_Light_Attributes(
    Color=<rgba 0 0.6738 1 1>
    Create Color Attribute=true
    Create Exponent Attribute=true
    Create Position Attribute=true
    Create Radius Attribute=true
    Exponent=5
    Position=<10 0 0>
    Radius=95
    Scale Color=&FloatFromCurve(
     FloatCurve=DataInterface() {propertyValues {Curve {
        defaultValue=1
        keys [{interpMode=RCIM_Cubic, tangentMode=RCTM_User, time=0, value=1}, {interpMode=RCIM_Cubic, tangentMode=RCTM_User, time=1, value=0}]
       }}}
     Scale Curve=10
    )
   )
  ]
  rendererClasses [
   &NiagaraMeshRendererProperties
   &NiagaraSpriteRendererProperties
   &NiagaraLightRendererProperties
  ]
  renderers [
   {rendererClass=&NiagaraMeshRendererProperties, rendererIndex=0}
   {rendererClass=&NiagaraSpriteRendererProperties, rendererIndex=1}
   {rendererClass=&NiagaraSpriteRendererProperties, rendererIndex=2}
   {rendererClass=&NiagaraSpriteRendererProperties, rendererIndex=3}
   {rendererClass=&NiagaraLightRendererProperties, rendererIndex=4}
   {rendererClass=&NiagaraSpriteRendererProperties, rendererIndex=5}
   {rendererClass=&NiagaraSpriteRendererProperties, rendererIndex=6}
  ]
 }
 CubeRing {
  emitterUpdateScript [
   &EmitterState()
   &SpawnBurst_Instantaneous(Spawn Count=13, Use Spawn Probability=false, UseLoopCountLimit=false) {module=SpawnBurst_Instantaneous001}
  ]
  particleSpawnScript [
   &InitializeParticle(
    Color Mode=::NewEnumerator0
    Lifetime=&RandomRangeFloat(Minimum=0.3, Maximum=0.5)
    Mesh Scale Mode=::NewEnumerator3
    Mesh Uniform Scale=&Multiply_Float(A=&RandomRangeFloat(Minimum=0.16, Maximum=0.19), B=$System.Global Scale)
    Sprite Size Mode=::NewEnumerator3
    UsePositionOffset=false
    Write Material Random=true
   )
   &ShapeLocation(
    MaskSpawnGroup=false
    Override Seed=false
    Ring / Disc Distribution Mode:ENiagaraSphereDistributionMode=::NewEnumerator1
    Ring Radius=&FN_FrontEndFloatModifier(Gameplay=0.6, Front End=2)
    Rotation Axis=&ReturnCameraVector(Camera Vectors=::NewEnumerator4, Coordinate Space=::Simulation, Origin Position=$Particles.Position)
    Rotation Mode=::NewEnumerator1
    Shape Primitive=::NewEnumerator4
    U Position=&ReturnNormalizedExecIndex()
    UseOwnerScale=true
   )
   &AddVelocity(
    Module.Use Falloff Distance=false
    Override Seed=false
    Velocity Mode=::NewEnumerator1
    Velocity Speed=&RandomRangeFloat(Minimum=75, Maximum=150)
   )
   &InitialMeshOrientation(
    Mesh Orientation Mode=::NewEnumerator2
    Rotation=&RandomRangeVector(Minimum=<0 -0.05 -0.05>, Maximum=<1 0.05 0.05>)
    Use Rotation Vector=true
   ) {enabled=false}
   &CameraOffset(Camera Offset Amount=5)
   &DynamicMaterialParameters(
    Index 0 Param 1=&RandomRangeFloat(Minimum=-1, Maximum=0)
    Module.Index0Float0Write=true
    Module.Index0Float1Write=true
    Module.Index0Float2Write=true
    Module.Index0Float3Write=true
   ) {enabled=false, module=DynamicMaterialParameters001}
  ]
  particleUpdateScript [
   &ParticleState()
   &ScaleMeshSize(Scale Factor=&Multiply_VectorByFloat(
     Vector=<1 1 1>
     Float=&FloatFromCurve(FloatCurve=DataInterface() {propertyValues {Curve {
         defaultValue=1
         keys [
          {arriveTangent=3.3604, interpMode=RCIM_Cubic, leaveTangent=3.3604, tangentMode=RCTM_User, time=0, value=0}
          {interpMode=RCIM_Cubic, tangentMode=RCTM_User, time=0.2919, value=1}
          {arriveTangentWeight=0.073, interpMode=RCIM_Cubic, leaveTangentWeight=0.3533, tangentMode=RCTM_User, time=0.5603, value=1}
          {arriveTangent=-4.8764, tangentMode=RCTM_User, time=1, value=0}
         ]
        }}})
    ))
   &Drag(Use Linear Drag=true, Use Rotational Drag=true)
   &GravityForce(Gravity=&Multiply_VectorByFloat(
     Vector=&RandomRangeVector(Minimum=<0 0 50>, Maximum=<0 0 150>)
     Float=&FN_FrontEndFloatModifier(Gameplay=1, Front End=0.4)
    ))
   &ScaleColor(
    Scale RGB=&Multiply_VectorByFloat(
     Vector=&MakeVectorFromLinearColorRGB(Color=&ColorFromCurve(ColorCurve=DataInterface() {propertyValues {
         AlphaCurve {keys [<at 0 1>]}
         BlueCurve {keys [<at 0 10>,<at 0.6589 1>]}
         GreenCurve {keys [<at 0 8.4899>,<at 0.6589 1>]}
         RedCurve {keys [<at 0 5.4292>,<at 0.6589 1>]}
        }}))
     Float=&MakeCustomFloatFromBool(Bool=$User.bIsFrontEndPreview, True Float=2, False Float=1)
    )
    ScaleA=true
    ScaleRGB=true
   )
   &CurlNoiseForce(
    Bias Noise=false
    Noise Strength=&FloatFromCurve(
     FloatCurve=DataInterface() {propertyValues {Curve {
        defaultValue=1
        keys [
         {arriveTangent=1.2977, interpMode=RCIM_Cubic, leaveTangent=1.2977, tangentMode=RCTM_User, time=0.1, value=0}
         {interpMode=RCIM_Cubic, tangentMode=RCTM_User, time=1, value=1}
        ]
       }}}
     Scale Curve=200
    )
    Pan Noise=false
   )
   &SolveForcesAndVelocity(Clamp Velocity=false, Limit Acceleration=false)
   &DynamicMaterialParameters(
    Index 0 Param 1=&FloatFromCurve(FloatCurve=DataInterface() {propertyValues {Curve {keys [<at 0.1966 -1>,<at 1 0>]}}})
    Module.Index0Float0Write=true
    Module.Index0Float1Write=true
    Module.Index0Float2Write=true
    Module.Index0Float3Write=true
   ) {enabled=false}
   &UpdateMeshOrientation(
    Rotation Rate=&FloatFromCurve(
     FloatCurve=DataInterface() {propertyValues {Curve {
        defaultValue=1
        keys [{interpMode=RCIM_Cubic, tangentMode=RCTM_User, time=0, value=0}, {interpMode=RCIM_Cubic, tangentMode=RCTM_User, time=0.5, value=1}]
       }}}
     Scale Curve=&RandomRangeFloat(Minimum=0.4, Maximum=1)
    )
    Rotation Vector=&RandomRangeVector(Minimum=<0 0 0>, Maximum=<1 1 1>)
   )
  ]
  rendererClasses [&NiagaraSpriteRendererProperties, &NiagaraMeshRendererProperties]
  renderers [
   {rendererClass=&NiagaraSpriteRendererProperties}
   {rendererClass=&NiagaraMeshRendererProperties, rendererIndex=1}
  ]
 }
 Cubes {
  bEnabled=false
  emitterUpdateScript [
   &EmitterState()
   &SpawnBurst_Instantaneous(Spawn Count=7, Use Spawn Probability=false, UseLoopCountLimit=false)
  ]
  particleSpawnScript [
   &InitializeParticle(
    Lifetime=&FloatFromCurve(
     FloatCurve=DataInterface() {propertyValues {Curve {keys [<at 0 0.4>,<at 0.75 0.57>,<at 1 0.8>]}}}
     CurveIndex=&ReturnNormalizedExecIndex()
    )
    Mesh Scale Mode=::NewEnumerator3
    Mesh Uniform Scale=&Multiply_Float(A=&RandomRangeFloat(Minimum=0.16, Maximum=0.21), B=$System.Global Scale)
    UsePositionOffset=false
    Write Material Random=true
   )
   &AddVelocity(
    Module.Use Falloff Distance=false
    Override Seed=false
    Velocity Mode=::NewEnumerator1
    Velocity Speed=&Multiply_Float(
     A=&RandomRangeFloat(Minimum=160, Maximum=250)
     B=&FN_FrontEndFloatModifier(Gameplay=1, Front End=0.6)
    )
   )
   &ShapeLocation(
    MaskSpawnGroup=false
    Sphere Distribution=::NewEnumerator3
    Sphere Radius=&FN_FrontEndFloatModifier(Gameplay=15, Front End=8)
    UseOwnerScale=true
   )
   &InitialMeshOrientation(
    Mesh Orientation Mode=::NewEnumerator2
    Rotation=&RandomRangeVector(Minimum=<0 -0.05 -0.05>, Maximum=<1 0.05 0.05>)
    Use Rotation Vector=true
   ) {enabled=false}
   &DynamicMaterialParameters(
    Index 0 Param 1=&RandomRangeFloat(Minimum=-1, Maximum=0)
    Module.Index0Float0Write=true
    Module.Index0Float1Write=true
    Module.Index0Float2Write=true
    Module.Index0Float3Write=true
   ) {enabled=false}
  ]
  particleUpdateScript [
   &Drag(
    Drag=&FloatFromCurve(
     FloatCurve=DataInterface() {propertyValues {Curve {keys [{interpMode=RCIM_Cubic, leaveTangent=3.6589, tangentMode=RCTM_Break, time=0, value=0.15}, <at 0.8 1>]}}}
     Scale Curve=6
    )
    Use Linear Drag=true
    Use Rotational Drag=false
   )
   &GravityForce(Gravity=&Multiply_VectorByFloat(
     Vector=&RandomRangeVector(Minimum=<0 0 200>, Maximum=<0 0 400>)
     Float=&FN_FrontEndFloatModifier(Gameplay=1, Front End=0.4)
    ))
   &SolveForcesAndVelocity(Clamp Velocity=false, Limit Acceleration=false)
   &ParticleState()
   &ScaleMeshSize(Scale Factor=&VectorFromFloat(Value=&FloatFromCurve(FloatCurve=DataInterface() {propertyValues {Curve {keys [
          {interpMode=RCIM_Cubic, leaveTangent=6.7871, tangentMode=RCTM_Break, time=0, value=0.5}
          {arriveTangentWeight=0.0945, interpMode=RCIM_Cubic, leaveTangentWeight=0.2793, tangentMode=RCTM_Break, time=0.2, value=1}
          {arriveTangent=-2.3616, tangentMode=RCTM_User, time=1, value=0}
         ]}}})))
   &UpdateMeshOrientation(Rotation Rate=&RandomRangeFloat(Minimum=0.4, Maximum=1), Rotation Vector=<1 1 1>)
   &ScaleColor(
    Scale RGB=&Multiply_VectorByFloat(
     Vector=&MakeVectorFromLinearColorRGB(Color=&ColorFromCurve(ColorCurve=DataInterface() {propertyValues {
         AlphaCurve {keys [<at 0 1>]}
         BlueCurve {keys [<at 0 10>,<at 0.6862 1>]}
         GreenCurve {keys [<at 0 8.4899>,<at 0.6862 1>]}
         RedCurve {keys [<at 0 5.4292>,<at 0.6862 1>]}
        }}))
     Float=&MakeCustomFloatFromBool(Bool=$User.bIsFrontEndPreview, True Float=2, False Float=1)
    )
    ScaleA=true
    ScaleRGB=true
   ) {module=ScaleColor001}
   /SetVariables(Particles.GlowColor=&ColorFromCurve(ColorCurve=DataInterface() {propertyValues {
       AlphaCurve {keys [<at 0.049 0.4 RCIM_Cubic>,<at 1 0 RCIM_Cubic>]}
       BlueCurve {keys [<at 0 2 RCIM_Cubic>,<at 1 0.4167 RCIM_Cubic>]}
       GreenCurve {keys [<at 0 0.8564 RCIM_Cubic>,<at 1 0.0401 RCIM_Cubic>]}
       RedCurve {keys [<at 0 0 RCIM_Cubic>,<at 1 0.0561 RCIM_Cubic>]}
      }})) {enabled=false, isSetParameters=true}
  ]
  rendererClasses [&NiagaraMeshRendererProperties]
  renderers [{rendererClass=&NiagaraMeshRendererProperties, rendererIndex=0}]
 }
 FN_PhysMatDebris {
  emitterUpdateScript [
   &EmitterState()
   &SpawnBurst_Instantaneous(
    Spawn Count=&RandomRangeInt(Minimum=5, Maximum=10)
    Use Spawn Probability=false
    UseLoopCountLimit=false
   )
  ]
  particleSpawnScript [
   /SetVariables(Particles.BiasedRandom=&FloatFromCurve(
     FloatCurve=DataInterface() {propertyValues {Curve {keys [
         <at 0.2 1 RCIM_Cubic>
         {arriveTangent=-3.3394, interpMode=RCIM_Cubic, leaveTangent=-3.3394, tangentMode=RCTM_User, time=0.5, value=0.5}
         <at 0.8 0 RCIM_Cubic>
        ]}}}
     CurveIndex=&RandomRangeFloat(Minimum=0, Maximum=1)
    )) {isSetParameters=true}
   &InitializeParticle(
    Color Mode=::NewEnumerator0
    Lifetime=&RandomRangeFloat(Minimum=0.5, Maximum=1.25)
    Mass=&Lerp_Float(A=0.6, B=1.2, Alpha=$Particles.BiasedRandom)
    Mass Mode=::NewEnumerator1
    Mesh Scale=&Multiply_VectorByFloat(
     Vector=&RandomRangeVector(Minimum=<0.9 0.4 0.5>, Maximum=<1 1 1>)
     Float=&Lerp_Float(A=0.7, B=1, Alpha=$Particles.BiasedRandom)
    )
    Mesh Scale Mode=::NewEnumerator4
    UsePositionOffset=false
    Write Material Random=true
   )
   &FN_PhysicalSurface(
    Dirt Mesh Count=2
    Glass Mesh Count=3
    Ice Mesh Count=2
    Metal Gravity=&RandomRangeFloat(Minimum=-350, Maximum=-1800)
    Metal Mesh Count=2
    Metal Velocity=&RandomRangeFloat(Minimum=80, Maximum=375)
    Metal Velocity Cone Angle=130
    Physical Surface=$User.PhysicalSurface
    Stone Gravity=&RandomRangeFloat(Minimum=-700, Maximum=-1250)
    Stone Mesh Count=2
    Stone Velocity=&RandomRangeFloat(Minimum=80, Maximum=550)
    Stone Velocity Cone Angle=120
    Wood Gravity=&RandomRangeFloat(Minimum=-450, Maximum=-900)
    Wood Mesh Count=2
    Wood Scale=1.5
    Wood Velocity=&RandomRangeFloat(Minimum=80, Maximum=600)
    Wood Velocity Cone Angle=130
   )
   &AddVelocityInCone(
    Cone Angle=$Output.FN_PhysicalSurface.SurfaceResponseConeAngle
    Module.Use Velocity Falloff On Cone Axis=false
    Velocity Strength=$Output.FN_PhysicalSurface.SurfaceResponseVelocity
   )
   &InitialMeshOrientation(Use Rotation Vector=false)
   &AddRotationalVelocity(Rotation Rate=&RandomVector(Vector Scale=&RandomRangeFloat(Minimum=1, Maximum=2)))
   &ApplyInitialForces()
   /SetVariables(Particles.GravityOverrideValue=$Output.FN_PhysicalSurface.GravityOverride) {isSetParameters=true}
  ]
  particleUpdateScript [
   &ParticleState()
   &GravityForce(Gravity=$Particles.GravityOverrideValue)
   &Collision(
    Advanced Aging Rate=1
    Age Colliding Particles=false
    CPU Collision Trace Channel=ECC_WorldStatic
    Collision Enabled=&Boolean_AndOperation(
     Bool1=&SetBoolByFloatComparison(A=$Particles.Initial.Lifetime, B=1)
     Bool2=&SetBoolByFloatComparison(A=$Particles.Age, B=0.12)
    )
    Control Roll On Collision=true
    Friction=1
    Friction During a Bounce=0.25
    Max CPU Trace Length=100
    Module.Control Rest State Via Surface Normal=false
    Module.EnableMaxCollisionCount=false
    Module.Randomize Collision Normal Vector=true
    Randomize Collision Normal=0.25
    ScaleRateOfRotation=false
   ) {enabled=false}
   &SolveForcesAndVelocity(Clamp Velocity=false, Limit Acceleration=false)
   &ScaleMeshSize(Scale Factor=&VectorFromFloat(Value=&FloatFromCurve(FloatCurve=DataInterface() {propertyValues {Curve {keys [<at 0.6 1 RCIM_Cubic>, {arriveTangent=-6.308, interpMode=RCIM_Cubic, leaveTangent=-6.308, tangentMode=RCTM_User, time=1, value=0}]}}})))
   &DynamicMaterialParameters(Module.Param0WriteEnabled=true, Module.Param1WriteEnabled=true, Module.Param2WriteEnabled=false, Module.Param3WriteEnabled=false)
  ]
  rendererClasses [&NiagaraSpriteRendererProperties, &NiagaraMeshRendererProperties]
  renderers [
   {rendererClass=&NiagaraSpriteRendererProperties}
   {rendererClass=&NiagaraMeshRendererProperties, rendererIndex=1}
   {rendererClass=&NiagaraMeshRendererProperties, rendererIndex=2}
   {rendererClass=&NiagaraMeshRendererProperties, rendererIndex=3}
   {rendererClass=&NiagaraMeshRendererProperties, rendererIndex=4}
   {rendererClass=&NiagaraMeshRendererProperties, rendererIndex=5}
   {rendererClass=&NiagaraMeshRendererProperties, rendererIndex=6}
   {rendererClass=&NiagaraMeshRendererProperties, rendererIndex=7}
   {rendererClass=&NiagaraMeshRendererProperties, rendererIndex=8}
  ]
 }
 Flare001 {
  emitterUpdateScript [
   /SetVariables(Transient.Shared.FloatFromCurve005_FloatCurve=DataInterface() {propertyValues {Curve {keys [
        {arriveTangent=-29.9725, interpMode=RCIM_Cubic, leaveTangent=-29.9725, tangentMode=RCTM_User, time=0, value=5}
        {interpMode=RCIM_Cubic, leaveTangent=-41.015, time=0.3, value=1}
       ]}}}) {isSetParameters=true}
   &EmitterState(Life Cycle Mode=::NewEnumerator1, Loop Behavior=::NewEnumerator1, Loop Delay=0.1, Loop Duration=1, UseLoopDelay=false)
   &SpawnBurst_Instantaneous(Spawn Count=2, Use Spawn Probability=false, UseLoopCountLimit=false)
   /SetVariables(Emitter.UV_Control:Vector4f=&MakeVector4(
     X=&FloatFromCurve(
      FloatCurve=?
      Transient.Shared.FloatFromCurve005_FloatCurve:NiagaraDataInterfaceCurve=DataInterface() {propertyValues {Curve {keys [
          {arriveTangent=-29.9725, interpMode=RCIM_Cubic, leaveTangent=-29.9725, tangentMode=RCTM_User, time=0, value=5}
          {interpMode=RCIM_Cubic, leaveTangent=-41.015, time=0.3, value=1}
         ]}}}
      CurveIndex=$Emitter.Age
     )
     Y=&FloatFromCurve(
      FloatCurve=?
      Transient.Shared.FloatFromCurve005_FloatCurve:NiagaraDataInterfaceCurve=DataInterface() {propertyValues {Curve {keys [
          {arriveTangent=-29.9725, interpMode=RCIM_Cubic, leaveTangent=-29.9725, tangentMode=RCTM_User, time=0, value=5}
          {interpMode=RCIM_Cubic, leaveTangent=-41.015, time=0.3, value=1}
         ]}}}
      CurveIndex=$Emitter.Age
     )
     Z=&Divide_Float(
      A=&Subtract_Float(
       A=&FloatFromCurve(
        FloatCurve=?
        Transient.Shared.FloatFromCurve005_FloatCurve:NiagaraDataInterfaceCurve=DataInterface() {propertyValues {Curve {keys [
            {arriveTangent=-29.9725, interpMode=RCIM_Cubic, leaveTangent=-29.9725, tangentMode=RCTM_User, time=0, value=5}
            {interpMode=RCIM_Cubic, leaveTangent=-41.015, time=0.3, value=1}
           ]}}}
        CurveIndex=$Emitter.Age
       )
       B=1
      )
      B=-2
     )
     W=&Divide_Float(
      A=&Subtract_Float(
       A=&FloatFromCurve(
        FloatCurve=?
        Transient.Shared.FloatFromCurve005_FloatCurve:NiagaraDataInterfaceCurve=DataInterface() {propertyValues {Curve {keys [
            {arriveTangent=-29.9725, interpMode=RCIM_Cubic, leaveTangent=-29.9725, tangentMode=RCTM_User, time=0, value=5}
            {interpMode=RCIM_Cubic, leaveTangent=-41.015, time=0.3, value=1}
           ]}}}
        CurveIndex=$Emitter.Age
       )
       B=1
      )
      B=-2
     )
    )) {isSetParameters=true}
  ]
  particleSpawnScript [
   &InitializeParticle(
    Color=&MergeHelper_ColorSwitch(
     0=<rgba 1 1 1 0.8>
     1=<rgba 1 1 1 0.4>
     2=?
     3=?
     Index=&ReturnParticleIDAttributes()
     Num=::NewEnumerator0
    )
    Lifetime=0.3
    Sprite Size Mode=::NewEnumerator3
    Uniform Sprite Size=&Multiply_Float(
     A=&MergeHelper_FloatSwitch(0=95, 1=180, 2=?, 3=?, Index=&ReturnParticleIDAttributes(), Num=::NewEnumerator0)
     B=0.8
    )
    UsePositionOffset=false
    Write Material Random=true
   )
   &DynamicMaterialParameters(
    Index 0 Param 1=&RandomRangeFloat(Minimum=0.4, Maximum=0.6)
    Index 0 Param 2=1
    Module.Param0WriteEnabled=true
    Module.Param1WriteEnabled=true
    Module.Param2WriteEnabled=true
    Module.Param3WriteEnabled=true
   )
   &CameraOffset(Camera Offset Amount=&MakeCustomFloatFromBool(Bool=$User.bIsFrontEndPreview, True Float=1, False Float=30))
   &SpriteFacingAndAlignment(WriteAlignment=false, WriteFacing=true)
  ]
  particleUpdateScript [
   &ParticleState()
   &ScaleColor(
    Scale Alpha=&Multiply_Float(
     A=1
     B=&FloatFromCurve(
      FloatCurve=DataInterface() {propertyValues {Curve {defaultValue=0, keys [<at 0.1 1 RCIM_Cubic>,<at 1 0 RCIM_Cubic>]}}}
      Scale Curve=&MakeCustomFloatFromBool(Bool=$User.bIsFrontEndPreview, True Float=0.7, False Float=1)
     )
    )
    Scale RGB=&MakeVectorFromLinearColorRGB(Color=&ColorFromCurve(ColorCurve=DataInterface() {propertyValues {
        AlphaCurve {keys [<at 0 1 RCIM_Cubic>]}
        BlueCurve {keys [
          <at 0.002 2.25>
          {arriveTangent=-2.4872, interpMode=RCIM_Cubic, leaveTangent=-2.4872, time=0.322, value=1.2}
          <at 0.5874 0.7939>
          <at 0.8621 0.5551 RCIM_Cubic>
         ]}
        GreenCurve {keys [
          <at 0.002 0>
          {arriveTangent=0.4103, interpMode=RCIM_Cubic, leaveTangent=0.4103, time=0.322, value=0}
          <at 0.5874 0.2402>
          <at 0.8621 0.5625 RCIM_Cubic>
         ]}
        RedCurve {keys [
          <at 0.002 1.5622>
          {arriveTangent=-2.4742, interpMode=RCIM_Cubic, leaveTangent=-2.4742, time=0.322, value=0.8332}
          <at 0.5874 0.1137>
          <at 0.8621 0 RCIM_Cubic>
         ]}
       }}))
    ScaleA=true
    ScaleRGB=true
   )
   &ScaleSpriteSize(
    Uniform Curve Scale=$System.Global Scale
    Uniform Curve Sprite Scale=DataInterface() {propertyValues {Curve {keys [{tangentMode=RCTM_User, time=0, value=0.5}, {tangentMode=RCTM_User, time=0.2, value=1}]}}}
   )
  ]
  rendererClasses [&NiagaraSpriteRendererProperties]
  renderers [{rendererClass=&NiagaraSpriteRendererProperties}]
 }
 Impact_Smoke001 {
  emitterUpdateScript [
   &EmitterState()
   &SpawnBurst_Instantaneous(
    Spawn Count=&RandomRangeInt(Minimum=2, Maximum=3)
    Use Spawn Probability=false
    UseLoopCountLimit=false
   )
  ]
  particleSpawnScript [
   &InitializeParticle(
    Color Maximum=<rgba 0.5885 1 1 0.15>
    Color Minimum=<rgba 0.6195 0.8421 1 0.1>
    Color Mode=::NewEnumerator2
    Lifetime Max=0.5
    Lifetime Min=0.4
    Lifetime Mode=::NewEnumerator1
    Sprite Rotation Mode=::NewEnumerator2
    Sprite Size Mode=::NewEnumerator3
    Uniform Sprite Size=&Multiply_Float(A=$System.Global Scale, B=&RandomRangeFloat(Minimum=115, Maximum=130))
    UsePositionOffset=false
    Write Material Random=true
   )
   &CameraOffset(Camera Offset Amount=&RandomRangeFloat(Minimum=40, Maximum=50, Evaluation Type=::NewEnumerator1))
   &ShapeLocation(
    Hemisphere Distribution=<1 0.5>
    MaskSpawnGroup=false
    Override Seed=false
    Rotation Angle=90
    Rotation Mode=::NewEnumerator1
    Sphere Radius=&MakeCustomFloatFromBool(Bool=$User.bIsFrontEnd, True Float=5, False Float=10)
    UseOwnerScale=true
   )
   &AddVelocity(Velocity=&Multiply_VectorByFloat(Vector=$Particles.ShapeLocation.ShapeVector, Float=10))
  ]
  particleUpdateScript [
   &ParticleState()
   &ScaleSpriteSize(Uniform Curve Sprite Scale=DataInterface() {propertyValues {Curve {keys [
        {arriveTangent=0.5339, interpMode=RCIM_Cubic, leaveTangent=0.5339, tangentMode=RCTM_User, time=0, value=0}
        {arriveTangent=0.5865, interpMode=RCIM_Cubic, leaveTangent=0.5865, tangentMode=RCTM_User, time=0.3904, value=0.7232}
        {arriveTangent=0.4342, interpMode=RCIM_Cubic, leaveTangent=0.4342, tangentMode=RCTM_User, time=1, value=1}
       ]}}})
   &ScaleColor(
    Scale Alpha=&FloatFromCurve(FloatCurve=DataInterface() {propertyValues {Curve {keys [{interpMode=RCIM_Cubic, leaveTangentWeight=0.2194, tangentMode=RCTM_User, time=0, value=0}, <at 0.25 0.5 RCIM_Cubic>, <at 1 0 RCIM_Cubic>]}}})
    Scale RGB=&VectorFromCurve(VectorCurve=DataInterface() {propertyValues {XCurve {keys [<at 0 1>,<at 1 1>]}, YCurve {keys [<at 0 1>,<at 1 1>]}, ZCurve {keys [<at 0 1>,<at 1 1>]}}})
    ScaleA=true
    ScaleRGB=true
   )
   &DynamicMaterialParameters(
    Index 0 Param 1=&Add_Float(
     A=&RandomRangeFloat(Minimum=0, Maximum=1)
     B=&FloatFromCurve(
      FloatCurve=DataInterface() {propertyValues {Curve {keys [
          <at 0 0 RCIM_Cubic>
          {arriveTangent=1.5947, interpMode=RCIM_Cubic, leaveTangent=1.5947, tangentMode=RCTM_User, time=0.2557, value=0.5257}
          <at 1 1 RCIM_Cubic>
         ]}}}
      CurveIndex=$Particles.Age
      Scale Curve=0.25
     )
    )
    Module.Param0WriteEnabled=true
    Module.Param1WriteEnabled=true
    Module.Param2WriteEnabled=true
    Module.Param3WriteEnabled=true
   ) {module=DynamicMaterialParameters001}
   &Drag(Use Linear Drag=true, Use Rotational Drag=true)
   &SolveForcesAndVelocity(Clamp Velocity=false, Limit Acceleration=false)
  ]
  rendererClasses [&NiagaraSpriteRendererProperties]
  renderers [
   {rendererClass=&NiagaraSpriteRendererProperties}
   {rendererClass=&NiagaraSpriteRendererProperties, rendererIndex=1}
  ]
 }
 Sparks {
  emitterUpdateScript [
   &EmitterState()
   &SpawnBurst_Instantaneous(Spawn Count=15, Use Spawn Probability=false, UseLoopCountLimit=false)
  ]
  particleSpawnScript [
   &InitializeParticle(
    Color Maximum=<rgba 0.0995 1.3418 2 1>
    Color Minimum=<rgba 0.0995 1.3418 2 1>
    Color Mode=::NewEnumerator2
    Lifetime Max=0.45
    Lifetime Min=0.1
    Lifetime Mode=::NewEnumerator1
    Sprite Size Mode=::NewEnumerator3
    Uniform Sprite Size=&Multiply_Float(A=$System.Global Scale, B=&RandomRangeFloat(Minimum=2, Maximum=4))
    UsePositionOffset=false
    Write Material Random=true
   )
   &ShapeLocation(
    MaskSpawnGroup=false
    Override Seed=false
    Sphere Radius=&Multiply_Float(A=2, B=&FN_FrontEndFloatModifier(Gameplay=0, Front End=0.6))
    UseOwnerScale=true
   )
   &CameraOffset(Camera Offset Amount=25)
   &AddVelocity(
    Module.Use Falloff Distance=false
    Override Seed=false
    Velocity Mode=::NewEnumerator1
    Velocity Speed=&Multiply_Float(A=&RandomRangeFloat(Minimum=400, Maximum=750), B=$System.Global Scale)
   )
  ]
  particleUpdateScript [
   &ParticleState()
   &AccelerationForce(Acceleration=&RandomRangeVector(Minimum=<0 0 -500>, Maximum=<0 0 -500>))
   &Drag(Drag=&RandomRangeFloat(Minimum=5, Maximum=8), Use Linear Drag=true, Use Rotational Drag=true)
   &SolveForcesAndVelocity(Clamp Velocity=false, Limit Acceleration=false)
   &ScaleSpriteSizeBySpeed(Max Scale Factor=<0.4 6>, Min Scale Factor=<0.5 1>)
   &ScaleSpriteSize(
    Uniform Curve Scale=&Multiply_Float(A=1, B=1)
    Uniform Curve Sprite Scale=DataInterface() {propertyValues {Curve {keys [{arriveTangent=-6.725, interpMode=RCIM_Cubic, leaveTangent=-6.725, tangentMode=RCTM_User, time=0.65, value=1}, <at 1 0 RCIM_Cubic>]}}}
   )
   &ScaleColor(
    Scale RGB=&VectorFromFloat(Value=&FloatFromCurve(
      FloatCurve=DataInterface() {propertyValues {Curve {keys [
          {interpMode=RCIM_Cubic, tangentMode=RCTM_User, time=0, value=1}
          <at 0.3 0.2>
          {interpMode=RCIM_Cubic, tangentMode=RCTM_User, time=1, value=0}
         ]}}}
      Scale Curve=10
     ))
    ScaleA=true
    ScaleRGB=true
   )
   /SetVariables(
    Particles.GlowColor=&MakeLinearColorFromVectorAndFloat(
     "Vector (RGB)"=&MakeVectorFromLinearColorRGB(Color=<rgba 0.298 0.7249 0.9047 1>)
     "Float (Alpha)"=&FloatFromCurve(FloatCurve=DataInterface() {propertyValues {Curve {keys [<at 0.5 1>,<at 1 0>]}}}, Scale Curve=0.08)
    )
    Particles.GlowSize=&Multiply_Vector2D(A=$Particles.SpriteSize, B=<3 2>)
   ) {isSetParameters=true}
  ]
  rendererClasses [&NiagaraSpriteRendererProperties]
  renderers [
   {rendererClass=&NiagaraSpriteRendererProperties}
   {rendererClass=&NiagaraSpriteRendererProperties, rendererIndex=1}
   {rendererClass=&NiagaraSpriteRendererProperties, rendererIndex=2}
  ]
 }
}
summary {
 systemName=NS_Pickaxe_MicStand_AnvilAlarmMoth_Impact
 emitters [
  {
   emitterName=FN_PhysMatDebris
   rendererClasses [&NiagaraSpriteRendererProperties, &NiagaraMeshRendererProperties]
  }
  {emitterName=Sparks, rendererClasses [&NiagaraSpriteRendererProperties]}
  {emitterName=Impact_Smoke001, rendererClasses [&NiagaraSpriteRendererProperties]}
  {bEnabled=false, emitterName=Cubes, rendererClasses [&NiagaraMeshRendererProperties]}
  {
   emitterName=Core_Flash
   rendererClasses [
    &NiagaraMeshRendererProperties
    &NiagaraSpriteRendererProperties
    &NiagaraLightRendererProperties
   ]
  }
  {emitterName=CubeRing, rendererClasses [&NiagaraSpriteRendererProperties, &NiagaraMeshRendererProperties]}
  {emitterName=Flare001, rendererClasses [&NiagaraSpriteRendererProperties]}
 ]
 userVariables [
  User.bIsFrontEnd {type=NiagaraBool}
  User.bIsFrontEndPreview {type=NiagaraBool}
  User.PhysicalSurface {type=NiagaraInt32}
 ]
}
